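# Material function for Athena landscape grass color matching that, as wired, passes its BaseColor input straight to the output unchanged.
# A side branch samples the terrain GroundColor texture at MF_Athena_Terrain_Coords UVs, multiplies it by BaseColor, then by ColormapDivisor from MPC_Landscape.
# That branch ends in an unconnected reroute, so the ground-color tint does not affect the result.
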MaterialFunction /Game/Athena/Environments/Landscape/MaterialFunctions/Standard/MF_Athena_GrassMatch
graph {nodes [
  n1 = FunctionInput() {inputName=BaseColor}
  FunctionOutput(None=@n1)
  n2 = TextureSample(UVs=MaterialFunctionCall!(WorldPosition) {materialFunction=/Game/Athena/Environments/Landscape/MaterialFunctions/Utilities/MF_Athena_Terrain_Coords}.Result) {
   samplerSource=SSM_Wrap_WorldGroupSettings
   samplerType=SAMPLERTYPE_LinearColor
   texture=/Game/Athena/Environments/Landscape/Textures/T_Athena_Terrain_GroundColor
  }
  Reroute(None=Multiply(
    A=Multiply(A=@n1, B=@n2.RGB)
    B=CollectionParameter() {collection=/Game/Athena/Environments/Landscape/MPC/MPC_Landscape, parameterName=ColormapDivisor}
   ))
 ]}
